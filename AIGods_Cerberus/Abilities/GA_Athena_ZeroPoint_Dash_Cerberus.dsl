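; Cerberus AI variant of the Zero Point dash ability, layered over a parent ability.
; Activation launches the pawn upward, waits briefly, triggers dash speech on its speech-bubble component, then defers to parent logic.
; It can activate when the blackboard target is within a distance band, or whenever the pawn has the cinematic tag.
; Overrides randomly angle the destination off the forward vector and use a large move distance for far targets, otherwise using parent results.
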(event Ability|EventActivateAbility
  (bind _outfortplayerpawn (CallFunction|GetAvatarasValidFortPlayerPawn)
    (:ValidFortPlayerPawn
      (bind _output_get (Variables|InternalData|SetAIPawn _outfortplayerpawn))
      (Character|LaunchCharacter _output_get "0.000000,0.000000,750.000000" false true)
      (Utilities|FlowControl|Delay 0.4)
      (Utilities|IsValid (Actor|GetComponentbyClass _outfortplayerpawn "/Script/FortniteGame.FortPawnComponent_SpeechBubble")
        (:Is%20Valid
          (SpeechBubble|StartSpeech (Actor|GetComponentbyClass _outfortplayerpawn "/Script/FortniteGame.FortPawnComponent_SpeechBubble") (Variables|Default|GetTagQuery_DashSpeech)))
        (:Is%20Not%20Valid
          _))
      (|CallParentFunction))
    (:AvatarCastFailed
      (Ability|EndAbility))))

(event Ability|EventOnEndAbility
  (params (bWasCancelled :type Boolean))
  (|CallParentFunction bWasCancelled) ; This node is disabled and will not be called.\nDrag off pins to build functionality.
)

(fn CalculateTargetLocation
  (return (select (>= (ScalableFloat|GetValueatLevel (Variables|Default|GetChanceToDeviate)) (Math|Random|RandomFloat)) (+ (Transformation|GetActorLocation (Variables|InternalData|GetAIPawn)) (* (Math|Vector|RotateVectorAroundAxis (Transformation|GetActorForwardVector (Variables|InternalData|GetAIPawn)) (Math|Random|RandomFloatinRange (* (ScalableFloat|GetValueatLevel (Variables|Default|GetRandomDegree)) -1.0) (ScalableFloat|GetValueatLevel (Variables|Default|GetRandomDegree))) (Transformation|GetActorUpVector (Variables|InternalData|GetAIPawn))) (Variables|Default|GetMoveDistance))) (|CallParentFunction))))

(fn GetMoveDistance
  (bind _largedistance (Variables|Default|GetLargeDistance))
  (bind _returnvalue (ScalableFloat|GetValueatLevel _largedistance))
  (bind _aipawn (Variables|InternalData|GetAIPawn))
  (bind _targetactor (Class|BPFLAICommonFunctions|GetTargetActorfromBlackboard _aipawn))
  (Utilities|IsValid _targetactor
    (:Is%20Valid
      (if (> (Transformation|GetDistanceTo _aipawn _targetactor) _returnvalue)
        (return _returnvalue)
        (else
          _)))
    (:Is%20Not%20Valid
      _))
  (return (|CallParentFunction)))

(fn CanActivateAbility
  (params (ActorInfo :type Gameplay%20Ability%20Actor%20Info%20Structure) (Handle :type Gameplay%20Ability%20Spec%20Handle%20Structure))
  (bind _returnvalue (CallFunction|GetAvatarasFortPlayerPawn))
  (Utilities|IsValid _returnvalue
    (:Is%20Valid
      (if (GameplayTags|HasMatchingGameplayTag _returnvalue "(TagName=\"Pawn.Athena.NPC.Cinematic\")")
        (return 0 true)
        (else
          (bind _targetactor (Class|BPFLAICommonFunctions|GetTargetActorfromBlackboard _returnvalue))
          (return 0 (and (>= (Math|Float|Absolute(Float) (Math|Vector|Distance(Vector) (Transformation|GetActorLocation _targetactor) (Transformation|GetActorLocation _returnvalue))) (ScalableFloat|GetValueatLevel (Variables|Default|GetMinimumDistance))) (<= (Math|Float|Absolute(Float) (Math|Vector|Distance(Vector) (Transformation|GetActorLocation _targetactor) (Transformation|GetActorLocation _returnvalue))) (ScalableFloat|GetValueatLevel (Variables|Default|GetMaximumDistance))))))))
    (:Is%20Not%20Valid
      (return 0 false))))
