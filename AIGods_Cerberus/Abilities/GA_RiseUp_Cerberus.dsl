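; GA_RiseUp_Cerberus is a gameplay ability for the Cerberus AI god boss, apparently its rise-up before combat.
; On activation it runs a target selection task and applies the gameplay effect container for Targeting.ApplicationTag.1 to the resulting targets.
; When the ability ends, cancelled or not, it casts the AI pawn to BP_PlayerPawn_NonParticipant_Boss and calls BossReadytoEngage on it.
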(event Ability|EventActivateAbility
  (|CallParentFunction)
  (bind (_asynctaskproxy _targetdata _applicationtag) (Ability|Tasks|DoTargetSelection "None" "(TagName=\"Targeting.ApplicationTag.1\")" true)
    (:Targeted
      (Ability|ApplyGameplayEffectContainer _targetdata _applicationtag))))

(event Ability|EventOnEndAbility
  (params (bWasCancelled :type Boolean))
  (|CallParentFunction bWasCancelled)
  (bind _asbp_player_pawn_non_participant_boss (Utilities|Casting|CastToBP_PlayerPawn_NonParticipant_Boss (Variables|InternalData|GetAIPawn)))
  (Class|BPPlayerPawnNonParticipantBoss|BossReadytoEngage _asbp_player_pawn_non_participant_boss))
