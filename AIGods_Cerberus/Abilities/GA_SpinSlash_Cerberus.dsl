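; Gameplay ability for the Cerberus AI spin slash attack that adds an activation range check, leaving activate and end logic to its parent.
; CanActivateAbility requires a valid Fort player pawn avatar, then allows activation if the blackboard target is within the scalable MaximumDistance.
; Avatars tagged as cinematic NPCs skip the distance check.
; The parent check supplies relevant tags, but its pass/fail result does not feed the returned boolean.
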(event Ability|EventActivateAbility
  (|CallParentFunction) ; This node is disabled and will not be called.\nDrag off pins to build functionality.
)

(event Ability|EventOnEndAbility
  (params (bWasCancelled :type Boolean))
  (|CallParentFunction bWasCancelled) ; This node is disabled and will not be called.\nDrag off pins to build functionality.
)

(fn CanActivateAbility
  (params (ActorInfo :type Gameplay%20Ability%20Actor%20Info%20Structure) (Handle :type Gameplay%20Ability%20Spec%20Handle%20Structure))
  (bind _returnvalue (CallFunction|GetAvatarasFortPlayerPawn))
  (bind _relevanttags (|CallParentFunction ActorInfo Handle))
  (Utilities|IsValid _returnvalue
    (:Is%20Valid
      (if (GameplayTags|HasMatchingGameplayTag _returnvalue "(TagName=\"Pawn.Athena.NPC.Cinematic\")")
        (return 0 true)
        (else
          (bind _targetactor (Class|BPFLAICommonFunctions|GetTargetActorfromBlackboard _returnvalue))
          (return _relevanttags (<= (Math|Float|Absolute(Float) (Math|Vector|Distance(Vector) (Transformation|GetActorLocation _targetactor) (Transformation|GetActorLocation _returnvalue))) (ScalableFloat|GetValueatLevel (Variables|Default|GetMaximumDistance)))))))
    (:Is%20Not%20Valid
      (return _relevanttags false))))
